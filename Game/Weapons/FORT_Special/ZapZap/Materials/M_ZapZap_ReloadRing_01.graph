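# Translucent unlit mesh-particle material for the ZapZap weapon's reload ring effect.
# A beam-reveal texture, panned by the PannerReveal dynamic parameter instead of time, is scaled by particle alpha and offset by Opacity Add for opacity.
# The same texture multiplies a Fresnel term that blends inner and outer colors, tinted by particle color and passed through an emissive adjustment function.
# Vertex normals scaled by the Thickness WPO dynamic parameter control ring thickness.
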Material /Game/Weapons/FORT_Special/ZapZap/Materials/M_ZapZap_ReloadRing_01
graph {
 properties {bUsedWithMeshParticles=true, blendMode=BLEND_Translucent, pixelDepthOffsetMode=PDOM_Legacy, shadingModel=MSM_Unlit, usedShadingModels=Unlit}
 nodes [
  n1 = TextureSample!(
   UVs=Panner(Time=DynamicParameter() {defaultValue=<rgba 0 0.2 1 1>, paramNames [Thickness WPO, PannerReveal, Opacity Add, Param4]}.PannerReveal) {speedX=1}
   Tex
   Level
   Apply View MipBias
  ) {
   constMipValue=0
   mipValueMode=TMVM_MipLevel
   samplerType=SAMPLERTYPE_LinearColor
   texture=/Game/Effects/Fort_Effects/Textures/Tile/T_Raygun_BeamReveal_01
  }
  Add(
   A=Multiply(A=ParticleColor().A, B=@n1.R)
   B=DynamicParameter() {defaultValue=<rgba 0 0.2 1 1>, paramNames [Thickness WPO, PannerReveal, Opacity Add, Param4]}.Opacity Add
  )
  n2 = VectorParameter() {parameterName=Color Inner, defaultValue=<rgba 0.0125 0.1869 0.5 1>}
  n3 = VectorParameter() {parameterName=Color Outer, defaultValue=<rgba 1.2286 1.9166 15 1>}
  n4 = Multiply(
   A=MaterialFunctionCall(
    Brightness=ScalarParameter() {defaultValue=1.5, parameterName=Fresnel Brightness}
    Exponent=ScalarParameter() {defaultValue=2, parameterName=Fresnel Exponent}
   ) {materialFunction=&MF_Fresnel_FX}.Fresnel
   B=@n1.R
  )
  MaterialFunctionCall(Emissive=Multiply(A=LinearInterpolate(A=@n2.RGB, B=@n3.RGB, Alpha=@n4), B=ParticleColor().RGB)) {materialFunction=&MF_FX_EmissiveAdjustments}
  Multiply(
   A=VertexNormalWS()
   B=DynamicParameter() {defaultValue=<rgba 0 0.2 1 1>, paramNames [Thickness WPO, PannerReveal, Opacity Add, Param4]}.Thickness WPO
  )
 ]
}
